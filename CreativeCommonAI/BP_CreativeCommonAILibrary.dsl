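; BP_CreativeCommonAILibrary is a Creative AI helper library providing a function to give a device actor a nav invoker component.
; TryAddNavInvokerOnDevice runs only with authority and checks whether the playlist supports on-demand navmesh.
; When it does, it adds a component of the supplied Fort AI nav invoker class to the device and returns it.
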(fn TryAddNavInvokerOnDevice
  (params (Device :type Actor%20Object%20Reference) (NavInvokerClass :type Fort%20AINav%20Invoker%20Component%20Class%20Reference))
  (if (Networking|HasAuthority Device)
    (bind _navmeshondemandenabled (CreativeAI|Navigation|IsPlaylistSupportNavMeshonDemand Device Device))
    (if _navmeshondemandenabled
      (bind _returnvalue (Game|AddComponentbyClass Device 0 NavInvokerClass))
      (return _returnvalue))))
